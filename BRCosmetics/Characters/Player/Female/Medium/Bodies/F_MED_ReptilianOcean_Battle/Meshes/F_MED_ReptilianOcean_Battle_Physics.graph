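# Physics asset for the female medium ReptilianOcean Battle cosmetic body, covering its back flaps and lower body.
# Three flap chains (center, left, right) each run from a small master capsule through three capsule segments, alongside larger pelvis, lower spine and thigh capsules.
# Chain joints have limited swing and twist, tightest at the side flap roots.
# Flap masters, spine and thighs link to the pelvis with angularly free constraints.
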PhysicsAsset /BRCosmetics/Characters/Player/Female/Medium/Bodies/F_MED_ReptilianOcean_Battle/Meshes/F_MED_ReptilianOcean_Battle_Physics
bodies {
 dyn_back_flap_center_01 [{shapeType=Capsule, length=5, radius=2.5, shapeName=dyn_back_flap_center_01_capsule, center=<4 0 1.5>, rotation=<rot 90 0 0>}]
 dyn_back_flap_center_02 [{shapeType=Capsule, length=4.5, radius=3, shapeName=dyn_back_flap_center_02_capsule, center=<3 0 1.5>, rotation=<rot 90 0 0>}]
 dyn_back_flap_center_03 [{shapeType=Capsule, length=7, radius=3, shapeName=dyn_back_flap_center_03_capsule, center=<3 0 1>, rotation=<rot 90 0 0>}]
 dyn_back_flap_center_mstr [{shapeType=Capsule, length=0.505, radius=0.505, shapeName=dyn_back_flap_center_mstr_capsule, rotation=<rot 0 -90 0>}]
 dyn_back_flap_left_01 [{shapeType=Capsule, length=3.5, radius=1.5, shapeName=dyn_back_flap_left_01_capsule, center=<2 0 0>, rotation=<rot 90 0 0>}]
 dyn_back_flap_left_02 [{shapeType=Capsule, length=5, radius=2.5, shapeName=dyn_back_flap_left_02_capsule, center=<4 0 1.5>, rotation=<rot 90 0 0>}]
 dyn_back_flap_left_03 [{shapeType=Capsule, length=4.5, radius=3, shapeName=dyn_back_flap_left_03_capsule, center=<3 0 1.5>, rotation=<rot 84.3752 180 180>}]
 dyn_back_flap_left_mstr [{shapeType=Capsule, length=0.505, radius=0.505, shapeName=dyn_back_flap_left_mstr_capsule, rotation=<rot 0 -90 0>}]
 dyn_back_flap_right_01 [{shapeType=Capsule, length=3.5, radius=1.5, shapeName=dyn_back_flap_right_01_capsule, center=<2 0 0>, rotation=<rot 90 0 0>}]
 dyn_back_flap_right_02 [{shapeType=Capsule, length=5, radius=2.5, shapeName=dyn_back_flap_right_02_capsule, center=<4 0 1.5>, rotation=<rot 90 0 0>}]
 dyn_back_flap_right_03 [{shapeType=Capsule, length=4.5, radius=3, shapeName=dyn_back_flap_right_03_capsule, center=<3 0 1.5>, rotation=<rot 84.3752 -180 -180>}]
 dyn_back_flap_right_mstr [{shapeType=Capsule, length=0.505, radius=0.505, shapeName=dyn_back_flap_right_mstr_capsule, rotation=<rot 0 -90 0>}]
 pelvis [{shapeType=Capsule, length=13, radius=10, center=<-5 -2 0>}]
 spine_01 [{shapeType=Capsule, length=14.505, radius=7.005, shapeName=spine_01_capsule, center=<-4.6608 -1.6475 0>, rotation=<rot 0 -90 0>}]
 thigh_l [{shapeType=Capsule, length=18, radius=8.5, shapeName=thigh_l_capsule, center=<-18 -1.5 0>, rotation=<rot 90 180 -175>}]
 thigh_r [{shapeType=Capsule, length=18, radius=8.5, shapeName=thigh_r_capsule, center=<18 1.5 0>, rotation=<rot 90 90 95>}]
}
constraints [
 {bone1Name=dyn_back_flap_center_01, bone2Name=dyn_back_flap_center_mstr, swing1LimitDegrees=10, swing2LimitDegrees=45, twistLimitDegrees=5}
 {bone1Name=dyn_back_flap_center_02, bone2Name=dyn_back_flap_center_01, swing1LimitDegrees=10, swing2LimitDegrees=45, twistLimitDegrees=5}
 {bone1Name=dyn_back_flap_center_03, bone2Name=dyn_back_flap_center_02, swing1LimitDegrees=10, swing2LimitDegrees=45, twistLimitDegrees=5}
 {bone1Name=dyn_back_flap_left_01, bone2Name=dyn_back_flap_left_mstr, swing1LimitDegrees=2, swing2LimitDegrees=20, twistLimitDegrees=0}
 {bone1Name=dyn_back_flap_left_02, bone2Name=dyn_back_flap_left_01, swing1LimitDegrees=10, swing2LimitDegrees=45, twistLimitDegrees=5}
 {bone1Name=dyn_back_flap_left_03, bone2Name=dyn_back_flap_left_02, swing1LimitDegrees=10, swing2LimitDegrees=45, twistLimitDegrees=5}
 {bone1Name=dyn_back_flap_right_01, bone2Name=dyn_back_flap_right_mstr, swing1LimitDegrees=2, swing2LimitDegrees=20, twistLimitDegrees=0}
 {bone1Name=dyn_back_flap_right_02, bone2Name=dyn_back_flap_right_01, swing1LimitDegrees=10, swing2LimitDegrees=45, twistLimitDegrees=5}
 {bone1Name=dyn_back_flap_right_03, bone2Name=dyn_back_flap_right_02, swing1LimitDegrees=10, swing2LimitDegrees=45, twistLimitDegrees=5}
 {
  bone1Name=spine_01
  bone2Name=pelvis
  swing1LimitDegrees=45
  swing1Motion=Free
  swing2LimitDegrees=45
  swing2Motion=Free
  twistLimitDegrees=45
  twistMotion=Free
 }
 {
  bone1Name=thigh_l
  bone2Name=pelvis
  swing1LimitDegrees=45
  swing1Motion=Free
  swing2LimitDegrees=45
  swing2Motion=Free
  twistLimitDegrees=45
  twistMotion=Free
 }
 {
  bone1Name=thigh_r
  bone2Name=pelvis
  swing1LimitDegrees=45
  swing1Motion=Free
  swing2LimitDegrees=45
  swing2Motion=Free
  twistLimitDegrees=45
  twistMotion=Free
 }
 {
  bone1Name=dyn_back_flap_left_mstr
  bone2Name=pelvis
  swing1LimitDegrees=45
  swing1Motion=Free
  swing2LimitDegrees=45
  swing2Motion=Free
  twistLimitDegrees=45
  twistMotion=Free
 }
 {
  bone1Name=dyn_back_flap_right_mstr
  bone2Name=pelvis
  swing1LimitDegrees=45
  swing1Motion=Free
  swing2LimitDegrees=45
  swing2Motion=Free
  twistLimitDegrees=45
  twistMotion=Free
 }
 {
  bone1Name=dyn_back_flap_center_mstr
  bone2Name=pelvis
  swing1LimitDegrees=45
  swing1Motion=Free
  swing2LimitDegrees=45
  swing2Motion=Free
  twistLimitDegrees=45
  twistMotion=Free
 }
]
