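# Post-process style AnimBP for the female ReptilianOcean Battle body, copying the parent body pose plus head-mesh facial bones.
# A bool blends a skydive/emote/front-end branch with a gameplay branch, each combining Control Rig and right shoulder armor corrections.
# RigidBody simulation follows, then Trail controllers and additive rotations lift back flaps by BwdMovement.
# The event graph derives alphas from bone deltas, weapon type, movement and player state.
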AnimBlueprint /BRCosmetics/Characters/Player/Female/Medium/Bodies/F_MED_ReptilianOcean_Battle/Meshes/F_MED_ReptilianOcean_Battle_AnimBP
target_skeleton=&/F_MED_ReptilianOcean_Battle_Skeleton
graphs [
 AnimGraph {nodes [
   n1 = Layered blend per bone(
    BasePose=Copy Pose From Mesh() {bUseAttachedParent=true}
    BlendPoses_0=Copy Pose From Mesh(SourceMeshComponent=Get head_mesh() -> (head_mesh:Skeletal Mesh Component Object Reference))
   ) {layerSetup [{branchFilters [
       {boneName=FACIAL_C_UnderChin}
       {boneName=FACIAL_L_UnderChin}
       {boneName=FACIAL_R_UnderChin}
       {boneName=FACIAL_R_Jawline}
       {boneName=FACIAL_C_Jawline}
       {boneName=FACIAL_L_Jawline}
       {boneName=FACIAL_C_FacialRoot}
       {boneName=FACIAL_C_Jaw}
       {boneName=FACIAL_C_Chin}
       {boneName=FACIAL_L_ChinSide}
       {boneName=FACIAL_R_ChinSide}
       {boneName=FACIAL_L_CheekLower}
       {boneName=FACIAL_L_Ear}
       {boneName=FACIAL_L_Masseter}
       {boneName=FACIAL_R_Ear}
       {boneName=FACIAL_R_CheekLower}
       {boneName=FACIAL_R_Masseter}
      ]}]}
   n2 = Control Rig!(Source=@n1, Alpha, Parachute:double=Get parachute() -> (parachute:double)) -> (Pose) {
    alphaBoolBlend {blendOption=Linear}
    controlRigAssetReference {blueprintRigClass=&/F_MED_ReptilianOcean_Battle_CtrlRig.F_MED_ReptilianOcean_Battle_CtrlRig_C}
   }
   n3 = "Constraint - dyn_upperarm_shoulder_armor_r_mstr (R)"(ComponentPose=Local To Component(LocalPose=@n2), Alpha=1, ConstraintWeights_0=1) {
    alphaBoolBlend {blendOption=Linear}
    boneToModify {boneName=dyn_upperarm_shoulder_armor_r_mstr}
    constraintSetup [{transformType=Rotation, targetBone {boneName=upperarm_twist_01_r}}]
    constraintWeights [1]
   }
   n4 = "Transform (Modify) Bone - Bone: dyn_upperarm_shoulder_armor_r_mstr"(ComponentPose=@n3, Alpha=Get parachute(), Translation="0,0,0", Rotation="-5,0,0", Scale="1,1,1") {
    rotationMode=BMM_Additive
    rotationSpace=BCS_BoneSpace
    translationMode=BMM_Additive
    translationSpace=BCS_BoneSpace
    alphaBoolBlend {blendOption=Linear}
    alphaScaleBiasClamp {bInterpResult=true}
    boneToModify {boneName=dyn_upperarm_shoulder_armor_r_mstr}
    rotation {pitch=-5}
   }
   n5 = "Transform (Modify) Bone - Bone: dyn_upperarm_shoulder_armor_r_mstr"(ComponentPose=@n4, Alpha=Get Skydive() -> (Skydive:double), Translation="-1,-1,-1", Rotation="0,0,10", Scale="1,1,1") {
    rotationMode=BMM_Additive
    rotationSpace=BCS_BoneSpace
    translationMode=BMM_Additive
    translationSpace=BCS_BoneSpace
    alphaBoolBlend {blendOption=Linear}
    alphaScaleBiasClamp {bInterpResult=true}
    boneToModify {boneName=dyn_upperarm_shoulder_armor_r_mstr}
    rotation {roll=10}
    translation=<-1 -1 -1>
   }
   n6 = "Transform (Modify) Bone - Bone: dyn_upperarm_shoulder_armor_r_mstr"(ComponentPose=@n5, Alpha=Get armRollRightAlpha() -> (armRollRightAlpha:double), Translation="-1,0,-1", Rotation="-10,0,0", Scale="1,1,1") {
    rotationMode=BMM_Additive
    rotationSpace=BCS_BoneSpace
    translationMode=BMM_Additive
    translationSpace=BCS_BoneSpace
    alphaBoolBlend {blendOption=Linear}
    boneToModify {boneName=dyn_upperarm_shoulder_armor_r_mstr}
    rotation {pitch=-10}
    translation {x=-1, z=-1}
   }
   n7 = Get BwdMovement() -> (BwdMovement:double)
   n8 = "Transform (Modify) Bone - Bone: dyn_back_flap_right_mstr"(ComponentPose=@n6, Alpha=@n7, Translation="0,0,0", Rotation="0,0,40", Scale="1,1,1") {rotationMode=BMM_Additive, alphaBoolBlend {blendOption=Linear}, boneToModify {boneName=dyn_back_flap_right_mstr}, rotation {roll=40}}
   n9 = "Transform (Modify) Bone - Bone: dyn_back_flap_center_mstr"(ComponentPose=@n8, Alpha=@n7, Translation="0,0,0", Rotation="0,0,40", Scale="1,1,1") {rotationMode=BMM_Additive, alphaBoolBlend {blendOption=Linear}, boneToModify {boneName=dyn_back_flap_center_mstr}, rotation {roll=40}}
   n10 = "Transform (Modify) Bone - Bone: dyn_back_flap_left_mstr"(ComponentPose=@n9, Alpha=@n7, Translation="0,0,0", Rotation="0,0,40", Scale="1,1,1") {rotationMode=BMM_Additive, alphaBoolBlend {blendOption=Linear}, boneToModify {boneName=dyn_back_flap_left_mstr}, rotation {roll=40}}
   n11 = Component To Local(ComponentPose=@n10)
   n12 = Layered blend per bone(BasePose=Copy Pose From Mesh() {bUseAttachedParent=true}, BlendPoses_0=Copy Pose From Mesh(SourceMeshComponent=Get head_mesh())) {layerSetup [{branchFilters [
       {boneName=FACIAL_C_UnderChin}
       {boneName=FACIAL_L_UnderChin}
       {boneName=FACIAL_R_UnderChin}
       {boneName=FACIAL_R_Jawline}
       {boneName=FACIAL_C_Jawline}
       {boneName=FACIAL_L_Jawline}
       {boneName=FACIAL_C_FacialRoot}
       {boneName=FACIAL_C_Jaw}
       {boneName=FACIAL_C_Chin}
       {boneName=FACIAL_L_ChinSide}
       {boneName=FACIAL_R_ChinSide}
       {boneName=FACIAL_L_CheekLower}
       {boneName=FACIAL_L_Ear}
       {boneName=FACIAL_L_Masseter}
       {boneName=FACIAL_R_Ear}
       {boneName=FACIAL_R_CheekLower}
       {boneName=FACIAL_R_Masseter}
      ]}]}
   n13 = Control Rig!(
    Source=@n12
    Alpha
    Pistol:double=Get pistol() -> (pistol:double)
    Melee:double=Get Melee() -> (Melee:double)
    Crouch:bool=Get bIsCrouching()
    Rifle:double=Get Rifle() -> (Rifle:double)
   ) -> (Pose) {
    alphaBoolBlend {blendOption=Linear}
    controlRigAssetReference {blueprintRigClass=&F_MED_ReptilianOcean_Battle_CtrlRig_C}
   }
   n14 = Look At - Bone: dyn_upperarm_shoulder_armor_r_mstr!(ComponentPose=Local To Component(LocalPose=@n13), bAlphaBoolEnabled:bool=True) -> (Pose) {
    alphaInputType=Bool
    alphaBoolBlend {blendOption=Linear}
    boneToModify {boneName=dyn_upperarm_shoulder_armor_r_mstr}
    lookAtLocation {x=0, z=-10}
    lookAtTarget {boneReference {boneName=lowerarm_r}}
    lookAt_Axis {axis=<-1 0>}
   }
   n15 = "Transform (Modify) Bone - Bone: dyn_upperarm_shoulder_armor_r_mstr"(ComponentPose=@n14, Alpha=Get armRestRightAlpha() -> (armRestRightAlpha:double), Translation="-0.5,0,1", Rotation="-5,5,0", Scale="1,1,1") {
    rotationMode=BMM_Additive
    rotationSpace=BCS_BoneSpace
    translationMode=BMM_Additive
    alphaBoolBlend {blendOption=Linear}
    boneToModify {boneName=dyn_upperarm_shoulder_armor_r_mstr}
    rotation {pitch=-5, yaw=5}
    translation {x=-0.5, z=1}
   }
   n16 = "Transform (Modify) Bone - Bone: dyn_upperarm_shoulder_armor_r_mstr"(ComponentPose=@n15, Alpha=1, Translation="0,0,0", Rotation="-15,0,0", Scale="1,1,1") {
    rotationMode=BMM_Additive
    rotationSpace=BCS_BoneSpace
    alphaBoolBlend {blendOption=Linear}
    boneToModify {boneName=dyn_upperarm_shoulder_armor_r_mstr}
    rotation {pitch=-15}
   }
   n17 = "Transform (Modify) Bone - Bone: dyn_upperarm_shoulder_armor_r_mstr"(ComponentPose=@n16, Alpha=Get armUpRightAlpha() -> (armUpRightAlpha:double), Translation="1,0,0", Rotation="20,-10,25", Scale="1,1,1") {
    rotationMode=BMM_Additive
    rotationSpace=BCS_BoneSpace
    translationMode=BMM_Additive
    translationSpace=BCS_BoneSpace
    alphaBoolBlend {blendOption=Linear}
    boneToModify {boneName=dyn_upperarm_shoulder_armor_r_mstr}
    rotation=<rot 20 -10 25>
    translation {x=1}
   }
   n18 = "Transform (Modify) Bone - Bone: dyn_upperarm_shoulder_armor_r_mstr"(ComponentPose=@n17, Alpha=Get armRollRightAlpha(), Translation="0,0,0", Rotation="0,0,0", Scale="1,1,1") {
    rotationMode=BMM_Additive
    rotationSpace=BCS_BoneSpace
    alphaBoolBlend {blendOption=Linear}
    boneToModify {boneName=dyn_upperarm_shoulder_armor_r_mstr}
   }
   n19 = Component To Local(ComponentPose=@n18)
   n20 = Get ParachuteOrEmoteFE() -> (ParachuteOrEmoteFE:bool)
   n21 = RigidBody!(
    ComponentPose=Local To Component(LocalPose=Blend Poses by bool(BlendPose_0=@n11, BlendPose_1=@n19, BlendTime_0=1, BlendTime_1=1, bActiveValue=@n20) {blendType=Linear})
    Alpha
    OverrideWorldGravity:Vector="0,0,0"
    ExternalForce
   ) -> (Pose) {
    bOverrideWorldGravity=true
    overridePhysicsAsset=&/F_MED_ReptilianOcean_Battle_Physics
    simulationSpace=BaseBoneSpace
    alphaBoolBlend {blendOption=Linear}
    baseBoneRef {boneName=pelvis}
    componentAppliedLinearAccClamp=<1500 1500 1500>
    componentLinearAccScale=<0.35 0.35 0.35>
    componentLinearVelScale=<0.35 0.35 0.35>
   }
   n22 = Get BwdMovement()
   n23 = Trail controller - Bone: dyn_back_flap_center_03(ComponentPose=@n21, Alpha=@n22) {
    bUsePlanarLimit=true
    chainLength=3
    debugLifeTime=0.1
    maxDeltaTime=0.333
    alphaScaleBiasClamp {bInterpResult=true, scale=0.5}
    baseJoint {boneName=root}
    planarLimits [{drivingBone {boneName=root}, planeTransform {location=<0 -3 0>, rotation=<rot 0 0 -90>, scale=<1 1 1>}}]
    rotationLimits [{limitMax=<180 180 180>}, {limitMax=<180 180 180>}, {limitMax=<180 180 180>}]
    rotationOffsets [<0 0 0>, <0 0 0>, <0 0 0>]
    trailBone {boneName=dyn_back_flap_center_03}
    trailRelaxationSpeed {editorCurveData {keys [<at -0.0134 19.3841 RCIM_Linear>,<at 0.5079 12.8986 RCIM_Linear>,<at 0.9955 9.6044 RCIM_Linear>]}}
   }
   n24 = Trail controller - Bone: dyn_back_flap_right_03(ComponentPose=@n23, Alpha=@n22) {
    bUsePlanarLimit=true
    debugLifeTime=0.1
    maxDeltaTime=0.333
    alphaScaleBiasClamp {bInterpResult=true, scale=0.5}
    baseJoint {boneName=root}
    planarLimits [{drivingBone {boneName=root}, planeTransform {location=<0 -3 0>, rotation=<rot 0 0 -90>, scale=<1 1 1>}}]
    rotationLimits [{limitMax=<180 180 180>}, {limitMax=<180 180 180>}]
    rotationOffsets [<0 0 0>, <0 0 0>]
    trailBone {boneName=dyn_back_flap_right_03}
    trailRelaxationSpeed {editorCurveData {keys [<at -0.0134 19.3841 RCIM_Linear>,<at 0.5079 12.8986 RCIM_Linear>,<at 0.9955 9.6044 RCIM_Linear>]}}
   }
   n25 = Trail controller - Bone: dyn_back_flap_left_03(ComponentPose=@n24, Alpha=@n22) {
    bUsePlanarLimit=true
    debugLifeTime=0.1
    maxDeltaTime=0.333
    alphaScaleBiasClamp {bInterpResult=true, scale=0.5}
    baseJoint {boneName=root}
    planarLimits [{drivingBone {boneName=root}, planeTransform {location=<0 -3 0>, rotation=<rot 0 0 -90>, scale=<1 1 1>}}]
    rotationLimits [{limitMax=<180 180 180>}, {limitMax=<180 180 180>}]
    rotationOffsets [<0 0 0>, <0 0 0>]
    trailBone {boneName=dyn_back_flap_left_03}
    trailRelaxationSpeed {editorCurveData {keys [<at -0.0134 19.3841 RCIM_Linear>,<at 0.5079 12.8986 RCIM_Linear>,<at 0.9955 9.6044 RCIM_Linear>]}}
   }
   n26 = "Transform (Modify) Bone - Bone: dyn_back_flap_right_mstr"(ComponentPose=@n25, Alpha=@n22, Translation="0,0,0", Rotation="0,0,25", Scale="1,1,1") {rotationMode=BMM_Additive, alphaBoolBlend {blendOption=Linear}, boneToModify {boneName=dyn_back_flap_right_mstr}, rotation {roll=25}}
   n27 = "Transform (Modify) Bone - Bone: dyn_back_flap_center_mstr"(ComponentPose=@n26, Alpha=@n22, Translation="0,0,0", Rotation="0,0,25", Scale="1,1,1") {rotationMode=BMM_Additive, alphaBoolBlend {blendOption=Linear}, boneToModify {boneName=dyn_back_flap_center_mstr}, rotation {roll=25}}
   n28 = "Transform (Modify) Bone - Bone: dyn_back_flap_left_mstr"(ComponentPose=@n27, Alpha=@n22, Translation="0,0,0", Rotation="0,0,25", Scale="1,1,1") {rotationMode=BMM_Additive, alphaBoolBlend {blendOption=Linear}, boneToModify {boneName=dyn_back_flap_left_mstr}, rotation {roll=25}}
   Output Pose(Result=Component To Local(ComponentPose=@n28)) {layerGroup=None}
  ]}
 EventGraph {nodes [
   Event BlueprintUpdateAnimation()
   > n1 = Sequence!(execute) -> (then_0, then_1, then_2:Exec, then_3:Exec, then_4:Exec, then_5:Exec)
   > n2 = Gravity Override State Function DEPRECATED(CharacterPartAnimInstance=Self-Reference(), GameTimeInSeconds=GetGameTimeInSeconds(), GravityOverrideParams=Get Gravity Override Params())
   > Set GravityOverride(GravityOverride=@n2)
   Set Skydive(execute=@n1.then_1, Skydive:double=SelectFloat(A=1, B=0, bPickA=Get bIsSkydiveFloating())) -> (Output_Get:double)
   > Set parachute(parachute:double=SelectFloat(A=1, B=0, bPickA=Get bIsParachuteOpened())) -> (Output_Get:double)
   > Set ParachuteOrEmoteFE(ParachuteOrEmoteFE:bool=OR Boolean!(A=Get bIsSkyDiving(), B=Get bIsPlayingEmote(), C:bool=Get bIsInFrontEnd()) -> (ReturnValue)) -> (Output_Get:bool)
   > Set Crouch(Crouch:double=SelectFloat(A=1, B=0, bPickA=Get bIsCrouching())) -> (Output_Get:double)
   > n3 = Set Skydiving(Skydiving:bool=OR Boolean(A=Get bIsSkyDiving(), B=Get bIsSkydiveFloating(), C=Get bIsSkydiveDiveMode())) -> (Output_Get:bool)
   > Branch(Condition=OR Boolean!(A=@n3, B=Get bIsPlayingEmote()) -> (ReturnValue))
   > Set aimUpAlpha(aimUpAlpha:double=0) -> (Output_Get:double)
   > Set neckAlpha(neckAlpha:double=0) -> (Output_Get:double)
   > Set armRestRightAlpha(armRestRightAlpha:double=0) -> (Output_Get:double)
   > Set armUpRightAlpha(armUpRightAlpha:double=0) -> (Output_Get:double)
   > Branch(Condition=@n3)
   > Set armRollRightAlpha(armRollRightAlpha:double=0) -> (Output_Get:double)
   Set BwdMovement(execute=@n1.then_5, BwdMovement:double=0) -> (Output_Get:double)
   > n4 = Branch(Condition=Get bRigidBodyEnabled())
   n5 = PlayerMovement(execute=@n4.else, AnimInstance=Self-Reference(), XDotProductCoefficient=1, YDotProductCoefficient=1, ZDotProductCoefficient=1)
   > n6 = Branch(Condition=Get bIsDBNOCarried())
   > Set BwdMovement(BwdMovement=0.5)
   n7 = Branch(execute=@n6.else, Condition=AND Boolean(A=Get bIsTargeting(), B=Get bIsCrouching()))
   n8 = MapRangeClamped(
    Value=float + float(A="Absolute (Float)"(A=@n5), B=MapRangeClamped(Value=@n5.YMovement, InRangeA=-1, InRangeB=1, OutRangeA=0.3)) {is=title K2Node_CommutativeAssociativeBinaryOperator}
    InRangeA=0
    InRangeB=1.65
   )
   > Set BwdMovement(BwdMovement=float * float(A=@n8, B=0.25) {is=title K2Node_PromotableOperator})
   Set BwdMovement(execute=@n7.else, BwdMovement=@n8)
   n9 = GetOwnerSkeletalMeshComponent()
   n10 = BreakRotator(InRot=BreakTransform(InTransform=GetDeltaTransformFromRefPose(self=@n9, BoneName=head, BaseName=spine_05)).Rotation)
   Set aimUpAlpha(execute=@n1.then_2, aimUpAlpha=MapRangeClamped(Value=@n10.Pitch, InRangeA=20, InRangeB=45))
   > Set neckAlpha(neckAlpha=MapRangeClamped(Value=@n10.Pitch, InRangeA=0, InRangeB=45))
   n11 = GetOwnerSkeletalMeshComponent()
   n12 = BreakRotator(InRot=BreakTransform(InTransform=GetDeltaTransformFromRefPose(self=@n11, BoneName=upperarm_r, BaseName=clavicle_r)).Rotation)
   > Set armUpRightAlpha(armUpRightAlpha=MapRangeClamped(Value=@n12.Yaw, InRangeA=30, InRangeB=115))
   > Set armRestRightAlpha(armRestRightAlpha=MapRangeClamped(Value=@n12.Yaw, InRangeA=30, InRangeB=-7))
   > Set armRollRightAlpha(armRollRightAlpha=MapRangeClamped(Value=@n12, InRangeA=0, InRangeB=-70))
   n13 = Cast To PlayerPawn_Athena(execute=@n1.then_3, Object=TryGetPawnOwner())
   > Set head_mesh(head_mesh:Skeletal Mesh Component Object Reference=GetSkeletalMeshForPartType(self=@n13, PartType=Head)) -> (Output_Get:Skeletal Mesh Component Object Reference)
   n14 = Cast To FortPlayerAnimInstance(execute=@n1.then_4, Object=GetAnimInstance(self=GetOwnerSkeletalMeshComponent()))
   > Set pistol(pistol:double=0) -> (Output_Get:double)
   > Set Melee(Melee:double=0) -> (Output_Get:double)
   > Set Rifle(Rifle:double=0) -> (Output_Get:double)
   > n15 = Switch on EFortWeaponCoreAnimation!(execute, Selection:EFortWeaponCoreAnimation Enum=Get WeaponCoreAnim(self=@n14)) -> (
    Melee:Exec
    Pistol:Exec
    Shotgun:Exec
    PaperBlueprint:Exec
    Rifle:Exec
    MeleeOneHand:Exec
    MachinePistol:Exec
    RocketLauncher:Exec
    GrenadeLauncher:Exec
    GoingCommando:Exec
    AssaultRifle:Exec
    TacticalShotgun:Exec
    SniperRifle:Exec
    TrapPlacement:Exec
    ShoulderLauncher:Exec
    AbilityDecoTool:Exec
    Crossbow:Exec
    C4:Exec
    RemoteControl:Exec
    DualWield:Exec
    AR_BullPup:Exec
    AR_ForwardGrip:Exec
    MedPackPaddles:Exec
    SMG_P90:Exec
    AR_DrumGun:Exec
    Consumable_Small:Exec
    Consumable_Large:Exec
    Balloon:Exec
    MountedTurret:Exec
    CreativeTool:Exec
    ExplosiveBow:Exec
    AshtonIndigo:Exec
    AshtonChicago:Exec
    MeleeDualWield:Exec
    Unarmed:Exec
   )
   Set Rifle(
    execute=[
     @n15.AR_BullPup
     @n15.AR_ForwardGrip
     @n15.AssaultRifle
     @n15.Crossbow
     @n15.GrenadeLauncher
     @n15.Rifle
     @n15.RocketLauncher
     @n15.Shotgun
     @n15.SniperRifle
     @n15.TacticalShotgun
    ]
    Rifle=1
   )
   Set Melee(execute=[@n15.Melee, @n15.MeleeDualWield], Melee=1)
   Set pistol(execute=@n15.Pistol, pistol=1)
   Event OnBeginSkydiving()
   > Set SkydivingAlpha(SkydivingAlpha:Rotator="65,0,0") -> (Output_Get:Rotator)
   Event OnEndSkydiving()
   > Set SkydivingAlpha(SkydivingAlpha="10,0,0")
   n16 = Branch(Condition=Get bIsSkyDiving())
   > n17 = Branch(Condition="float > float"(A=Get PawnVelocityZ(), B=50) {is=title K2Node_CallFunction})
   > Set BwdMovement(BwdMovement=float * float(A=Get BwdMovement(), B=0) {is=title K2Node_CommutativeAssociativeBinaryOperator})
   Set BwdMovement(execute=[@n16.else, @n17.else], BwdMovement=float * float(A=Get BwdMovement(), B=1))
  ]}
]
variables [
 aimUpAlpha {type=number}
 armRestRightAlpha {type=number}
 armRollRightAlpha {type=number}
 armUpRightAlpha {type=number}
 BwdMovement {type=number}
 Crouch {type=number}
 Gravity Override Params {type=GravityOverrideParamsStruct}
 GravityOverride {type=Vector}
 head_mesh {type=SkeletalMeshComponent}
 Melee {type=number}
 neckAlpha {type=number}
 parachute {type=number}
 ParachuteOrEmoteFE {type=boolean}
 pistol {type=number}
 Rifle {type=number}
 Skydive {type=number}
 Skydiving {type=boolean}
 SkydivingAlpha {type=Rotator}
]
